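; Blueprint for the Audio_Bass_A environment prop in the Helios set, which plays a destruction sound when it is destroyed.
; On BeginPlay it binds a custom event to the OnDeathPlayEffects delegate.
; That event plays the Destruction Sound variable at the actor's location, ignoring the damage and hit details it receives.
; Overlap and Tick events are empty, and the construction script only calls its parent.
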(event EventBeginPlay
  (EventDispatchers|BindEventtoOnDeathPlayEffectsDelegate (delegate Custom|OnDeathPlayEffectsDelegate_Event)))

(event Collision|EventActorBeginOverlap
  (params (OtherActor :type Actor%20Object%20Reference)))

(event Custom|OnDeathPlayEffectsDelegate_Event
  (params (Damage :type Float%20%28single-precision%29) (DamageTags :type Gameplay%20Tag%20Container%20Structure%20%28by%20ref%29) (Momentum :type Vector) (HitInfo :type Hit%20Result%20Structure%20%28by%20ref%29) (InstigatedBy :type Fort%20Pawn%20Object%20Reference) (DamageCauser :type Actor%20Object%20Reference) (EffectContext :type Gameplay%20Effect%20Context%20Handle%20Structure))
  (Audio|PlaySoundatLocation (Variables|Default|GetDestruction%20Sound) (Transformation|GetActorLocation)))

(event EventTick
  (params (DeltaSeconds :type Float%20%28single-precision%29)))

(fn ConstructionScript
  (|CallParentFunction))
